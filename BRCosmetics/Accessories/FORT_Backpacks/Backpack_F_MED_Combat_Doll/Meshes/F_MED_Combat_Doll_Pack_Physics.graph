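# Physics asset for the Combat Doll backpack cosmetic, defining simulation bodies for its dangling zipper bones.
# Four three-segment zipper chains each hang from a central pack body, every segment a small capsule constrained to the one before it.
# All joints share the same tight swing and twist limits, keeping the zipper motion restrained.
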PhysicsAsset /BRCosmetics/Accessories/FORT_Backpacks/Backpack_F_MED_Combat_Doll/Meshes/F_MED_Combat_Doll_Pack_Physics
bodies {
 dyn_zipper1_01 [{shapeType=Capsule, length=0, radius=0.8383, center=<1.0559 0.0025 0.3093>, rotation=<rot 0 -87.4143 -79.7479>}]
 dyn_zipper1_02 [{shapeType=Capsule, length=0, radius=0.8867, center=<0.9591 -0.0045 -0.0707>, rotation=<rot 0 90.1819 -82.5178>}]
 dyn_zipper1_03 [{shapeType=Capsule, length=0.8184, radius=1, center=<1.7472 0.0159 -0.1529>, rotation=<rot -2.4914 -90 -90.0873>}]
 dyn_zipper2_01 [{shapeType=Capsule, length=0.505, radius=0.505, center=<1 0 0>, rotation=<rot 0 -93.2699 -79.2087>}]
 dyn_zipper2_02 [{shapeType=Capsule, length=0.505, radius=0.505, center=<1 0 0>, rotation=<rot 0 -86.0414 -82.9685>}]
 dyn_zipper2_03 [{shapeType=Capsule, length=1.053, radius=0.9119, center=<1 0 0>, rotation=<rot -1.6413 -90 1.727>}]
 dyn_zipper3_01 [{shapeType=Capsule, length=0.505, radius=0.505, center=<-1 0 0>, rotation=<rot 0 -87.4144 -79.7483>}]
 dyn_zipper3_02 [{shapeType=Capsule, length=0.505, radius=0.505, center=<-1 0 0>, rotation=<rot 0 90.1819 -82.518>}]
 dyn_zipper3_03 [{shapeType=Capsule, length=1.0252, radius=0.7932, center=<-1 0 0>, rotation=<rot -2.4914 -90 -90.0874>}]
 dyn_zipper4_01 [{shapeType=Capsule, length=0.26, radius=0.8805, center=<-0.3816 -0.6084 0.4266>, rotation=<rot 0 -93.2699 -79.2058>}]
 dyn_zipper4_02 [{shapeType=Capsule, length=0.51, radius=1.2291, center=<-1.2184 -0.028 -0.0577>, rotation=<rot -1.2804 -75.5979 -83.085>}]
 dyn_zipper4_03 [{shapeType=Capsule, length=0.9119, radius=1.4791, center=<-2.1829 0.012 -0.0641>, rotation=<rot 81.0032 35.996 164.6321>}]
 pack [{shapeType=Capsule, length=1, radius=1}]
}
constraints [
 {bone1Name=dyn_zipper1_01, bone2Name=pack, swing1LimitDegrees=15, swing2LimitDegrees=25, twistLimitDegrees=5}
 {bone1Name=dyn_zipper1_02, bone2Name=dyn_zipper1_01, swing1LimitDegrees=15, swing2LimitDegrees=25, twistLimitDegrees=5}
 {bone1Name=dyn_zipper1_03, bone2Name=dyn_zipper1_02, swing1LimitDegrees=15, swing2LimitDegrees=25, twistLimitDegrees=5}
 {bone1Name=dyn_zipper2_01, bone2Name=pack, swing1LimitDegrees=15, swing2LimitDegrees=25, twistLimitDegrees=5}
 {bone1Name=dyn_zipper2_02, bone2Name=dyn_zipper2_01, swing1LimitDegrees=15, swing2LimitDegrees=25, twistLimitDegrees=5}
 {bone1Name=dyn_zipper2_03, bone2Name=dyn_zipper2_02, swing1LimitDegrees=15, swing2LimitDegrees=25, twistLimitDegrees=5}
 {bone1Name=dyn_zipper3_01, bone2Name=pack, swing1LimitDegrees=15, swing2LimitDegrees=25, twistLimitDegrees=5}
 {bone1Name=dyn_zipper3_02, bone2Name=dyn_zipper3_01, swing1LimitDegrees=15, swing2LimitDegrees=25, twistLimitDegrees=5}
 {bone1Name=dyn_zipper3_03, bone2Name=dyn_zipper3_02, swing1LimitDegrees=15, swing2LimitDegrees=25, twistLimitDegrees=5}
 {bone1Name=dyn_zipper4_01, bone2Name=pack, swing1LimitDegrees=15, swing2LimitDegrees=25, twistLimitDegrees=5}
 {bone1Name=dyn_zipper4_02, bone2Name=dyn_zipper4_01, swing1LimitDegrees=15, swing2LimitDegrees=25, twistLimitDegrees=5}
 {bone1Name=dyn_zipper4_03, bone2Name=dyn_zipper4_02, swing1LimitDegrees=15, swing2LimitDegrees=25, twistLimitDegrees=5}
]
